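# Sound cue for the outdoor ambient drill bit spot at the Asteria POI, a spatialized sizzle and steam bed with a reverb submix send.
# A mixer layers three continuous loops: drill sizzle, pitched-down coaster energy mech and large energy field.
# A fourth branch loops through a random delay, fade-in envelope and lowered-pitch modulator into a random pick of six reused steampunk weapon steam vents and exhausts, producing intermittent bursts.
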SoundCue /Game/Sounds/Asteria/POI/DrillBit/Spot_Drill_Bit_Sizzle_Steam_Loop_Cue
root=Mixer_0
nodes [
 {id=Delay_1, children [Enveloper_2], properties {delayMax=12, delayMin=3}}
 {
  id=Enveloper_2
  children [Modulator_1]
  properties {pitchCurve {editorCurveData {keys [{time=0}]}}, volumeCurve {editorCurveData {keys [<at 0 0>,<at 0.75 1>]}}}
 }
 {id=Looping_2, children [Delay_1]}
 {id=Mixer_0, children [Looping_2,WavePlayer_0,Modulator_0,WavePlayer_3], properties {inputVolume [0.5,1.5,1,1.5]}}
 {id=Modulator_0, children [WavePlayer_2], properties {pitchMax=0.8, pitchMin=0.8}}
 {id=Modulator_1, children [Random_1], properties {pitchMax=0.8, pitchMin=0.7}}
 {
  id=Random_1
  children [WavePlayer_7, WavePlayer_8, WavePlayer_9, WavePlayer_10, WavePlayer_4, WavePlayer_6]
  properties {weights [1,1,1,1,1,1]}
 }
 {id=WavePlayer_0, properties {bLooping=true, soundWaveAssetPtr=/Game/Sounds/Artemis/POI/Hollow_Base/Spots/Spot_Drill_Bit_Sizzle_Loop}}
 {id=WavePlayer_10, properties {soundWaveAssetPtr=/Game/Sounds/Fort_Weapons/SteamPunk/DrumAssault/SP_DrumAssault_Steam_Vent_03}}
 {
  id=WavePlayer_2
  properties {bLooping=true, soundWaveAssetPtr=/Game/Sounds/Artemis/POI/Hollow_Base_Takover/Spots/Spot_HollowBase_Coaster_Energy_Mech_Loop1}
 }
 {id=WavePlayer_3, properties {bLooping=true, soundWaveAssetPtr=/Game/Sounds/Apollo/Biomes/Skyfire/Energy_Field_Large_Close_Loop}}
 {id=WavePlayer_4, properties {soundWaveAssetPtr=/Game/Sounds/Fort_Weapons/SteamPunk/Assault/SP_Assault_SteamExhaust_03}}
 {id=WavePlayer_6, properties {soundWaveAssetPtr=/Game/Sounds/Fort_Weapons/SteamPunk/Assault/SP_Assault_SteamExhaust_02}}
 {id=WavePlayer_7, properties {soundWaveAssetPtr=/Game/Sounds/Fort_Weapons/SteamPunk/DrumAssault/SP_DrumAssault_Steam_Vent_04}}
 {id=WavePlayer_8, properties {soundWaveAssetPtr=/Game/Sounds/Fort_Weapons/SteamPunk/DrumAssault/SP_DrumAssault_Steam_Vent_02}}
 {id=WavePlayer_9, properties {soundWaveAssetPtr=/Game/Sounds/Fort_Weapons/SteamPunk/DrumAssault/SP_DrumAssault_Steam_Vent_01}}
]
properties {
 attenuationSettings=/Game/Sounds/Artemis/POI/Hollow_Base/Spots/Spot_Drill_Bit_Sizzle_Att
 maxDistance=3000
 soundClassObject=&Fort_Ambient_Outdoors_Athena
 volumeMultiplier=0.551
 soundSubmixSends [{minSendLevel=1, soundSubmix=&Fortnite_Reverb_Default}]
}
